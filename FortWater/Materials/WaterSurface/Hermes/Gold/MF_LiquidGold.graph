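# Material function that optionally blends a metallic liquid-gold layer over incoming water material attributes, also outputting a mobile water color and sphere mask.
# Panning and multi-scale masks, river flow maps, depth fade and an optional sphere mask drive coverage.
# Emissive adds two scales of noise-phased cell sparkles, sun glints and time-of-day glow.
# Parallax UVs, extra layers and detail normals drop out on cheaper quality or mobile paths.
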MaterialFunction /FortWater/Materials/WaterSurface/Hermes/Gold/MF_LiquidGold
graph {nodes [
  n1 = ScalarParameter() {defaultValue=0.02, group=Gold Liquid, parameterName=BaseGold}
  n2 = NamedRerouteUsage()
  n3 = ScalarParameter() {defaultValue=150, group=Gold Liquid, parameterName=Gold_Particles Size}
  n4 = ScalarParameter() {defaultValue=-0.02, group=Gold Liquid, parameterName=GoldPanSpeed2}
  n5 = ScalarParameter() {defaultValue=-0.03, group=Gold Liquid, parameterName=GoldPanSpeed2b}
  n6 = TextureSample(UVs=Panner(Coordinate=Divide(A=@n2, B=@n3), Speed=AppendVector(A=@n4, B=@n5)) {speedX=0.05, speedY=0.03}) {samplerSource=SSM_Wrap_WorldGroupSettings, samplerType=SAMPLERTYPE_Masks, texture=&/GoldLiquid_M}
  n7 = MaterialFunctionCall(Float3=@n6.RGB) {materialFunction=&BreakOutFloat3Components}
  n8 = Divide(A=NamedRerouteUsage(), B=@n3)
  n9 = TextureSample(UVs=Panner(
    Coordinate=NamedRerouteDeclaration(None=@n8) {name=UV_GoldDetail, nodeColor=<rgba 0 0.7891 1 1>}
    Speed=ScalarParameter() {defaultValue=0.02, group=Gold Liquid, parameterName=GoldPanSpeed}
   ) {speedX=0.05, speedY=0.03}) {samplerSource=SSM_Wrap_WorldGroupSettings, samplerType=SAMPLERTYPE_Masks, texture=&GoldLiquid_M}
  n10 = MaterialFunctionCall(Float3=@n9.RGB) {materialFunction=&BreakOutFloat3Components}
  n11 = Max(A=@n7.R, B=@n10.R)
  n12 = StaticSwitch(
   True=@n7.R
   False=ShadingPathSwitch(Default=QualitySwitch(Default=@n11, Low=@n7.R, High=@n10.R, Medium=@n7.R, Epic=@n11), Mobile=@n7.R) {desc=Migrated from FeatureLevelSwitch}
   Value=NamedRerouteUsage()
  )
  n13 = ScalarParameter() {defaultValue=0.7374, group=Gold Liquid, parameterName=GoldTContrast}
  n14 = ScalarParameter() {defaultValue=3225.6716, group=Gold Liquid, parameterName=Gold_Mask Size}
  n15 = TextureSample(UVs=Divide(A=NamedRerouteUsage(), B=@n14)) {samplerSource=SSM_Wrap_WorldGroupSettings, samplerType=SAMPLERTYPE_Masks, texture=&GoldLiquid_M}
  n16 = MaterialFunctionCall(Float3=@n15.RGB) {materialFunction=&BreakOutFloat3Components}
  n17 = NamedRerouteUsage()
  n18 = TextureSample(UVs=NamedRerouteDeclaration(None=Divide(A=@n17, B=@n14)) {name=UV_GoldLargeMask, nodeColor=<rgba 1 0 0.3281 1>}) {samplerSource=SSM_Wrap_WorldGroupSettings, samplerType=SAMPLERTYPE_Masks, texture=&GoldLiquid_M}
  n19 = Multiply(
   A=MaterialFunctionCall(Float3=@n18.RGB) {materialFunction=&BreakOutFloat3Components}.B
   B=ScalarParameter() {defaultValue=0.7, group=Gold Liquid, parameterName=Gold_LowerLayer_Darken}
  )
  n20 = Max(A=@n16.B, B=@n19)
  n21 = ShadingPathSwitch(Default=QualitySwitch(Default=@n20, Low=@n16.B, High=@n20, Medium=@n19, Epic=@n20), Mobile=@n16.B) {desc=Migrated from FeatureLevelSwitch}
  n22 = Divide(A=NamedRerouteUsage(), B=ScalarParameter() {defaultValue=10000, group=Gold Liquid, parameterName=Gold_MaskHuge Size})
  n23 = TextureSample(UVs=Panner(Coordinate=NamedRerouteDeclaration(None=@n22) {name=UV_GoldHugeMask, nodeColor=<rgba 1 0 0.3281 1>}) {speedX=0.01, speedY=0.005}) {samplerSource=SSM_Wrap_WorldGroupSettings, samplerType=SAMPLERTYPE_Masks, texture=&GoldLiquid_M}
  n24 = MaterialFunctionCall(Float3=@n23.A) {materialFunction=&BreakOutFloat3Components}
  n25 = Subtract(
   A=@n21
   B=LinearInterpolate(
    A=ScalarParameter() {defaultValue=-0.1844, group=Gold Liquid, parameterName=Gold_MaskT_1}
    B=ScalarParameter() {defaultValue=1.582, group=Gold Liquid, parameterName=Gold_MaskT_2}
    Alpha=@n24.B
   )
  )
  n26 = ScalarParameter() {defaultValue=0.25, group=Gold Liquid, parameterName=Gold_MaskFarCloseAdd}
  n27 = Power(Base=@n24.B, Exp=ScalarParameter() {defaultValue=1.3, group=Gold Liquid, parameterName=Gold_MaskFarContrast})
  n28 = ScalarParameter() {defaultValue=1.0994, group=Gold Liquid, parameterName=Gold_MaskFar}
  n29 = PixelDepth()
  n30 = ScalarParameter() {defaultValue=15000, parameterName=DistanceMaskBlend}
  n31 = LinearInterpolate(A=@n25, B=Add(A=Multiply(A=@n25, B=@n26), B=Multiply(A=@n27, B=@n28)), Alpha=Saturate(None=Divide(A=@n29.R, B=@n30)))
  n32 = ScalarParameter() {defaultValue=0.4, group=Gold Liquid, parameterName=Gold_MaskT_Hardness}
  n33 = Saturate(None=LinearInterpolate(B=ScalarParameter() {defaultValue=0.375, group=Gold Liquid, parameterName=Gold_MaskT_Max}, Alpha=Divide(A=@n31, B=@n32)))
  n34 = LinearInterpolate(A=@n12, B=Add(A=@n12, B=Power(Base=@n12, Exp=@n13)), Alpha=@n33)
  Add(A=@n1, B=@n34)
  CameraVectorWS()
  n35 = Divide(
   A=WorldPosition() {worldPositionShaderOffset=WPT_ExcludeAllShaderOffsets}.XYZ
   B=ScalarParameter() {defaultValue=70, parameterName=GoldCell_LargeSize}
  )
  n36 = ComponentMask(None=@n35) {g=true, r=true}
  n37 = Frac(None=@n36)
  n38 = Multiply(A=ScalarParameter() {defaultValue=0.1, parameterName=GoldSparkle_LargeSpeed}, B=Time())
  n39 = VectorNoise(World Position=@n35)
  n40 = Multiply(A=ComponentMask(None=@n39) {b=true, g=true, r=true}, B=ScalarParameter() {defaultValue=13516.0879, parameterName=GoldEmissive_Large Offset})
  n41 = Saturate(None=Sine(None=Add(A=@n38, B=@n40)))
  n42 = ComponentMask(None=@n41) {r=true}
  n43 = Multiply(A=ScalarParameter() {defaultValue=1, parameterName=GoldEmissiveSparkle_Large Radius}, B=@n42)
  n44 = Divide(A=Distance(A=@n37, B=Constant2Vector() {g=0.5, r=0.5}), B=@n43) {constB=0.5}
  Divide(A=Constant() {r=1}, B=Multiply(A=@n44, B=@n44))
  n45 = Divide(
   A=WorldPosition() {worldPositionShaderOffset=WPT_ExcludeAllShaderOffsets}.XYZ
   B=ScalarParameter() {defaultValue=40, parameterName=GoldCellSize}
  )
  n46 = ComponentMask(None=@n45) {g=true, r=true}
  n47 = Frac(None=@n46)
  n48 = Multiply(A=ScalarParameter() {defaultValue=0.1, parameterName=GoldSparkleSpeed}, B=Time())
  n49 = VectorNoise(World Position=@n45)
  n50 = Multiply(A=ComponentMask(None=@n49) {b=true, g=true, r=true}, B=ScalarParameter() {defaultValue=13516.0879, parameterName=GoldEmissive Offset})
  n51 = Saturate(None=Sine(None=Add(A=@n48, B=@n50)))
  n52 = ComponentMask(None=@n51) {r=true}
  n53 = Multiply(A=ScalarParameter() {defaultValue=1, parameterName=GoldEmissiveSparkle Radius}, B=@n52)
  n54 = Divide(A=Distance(A=@n47, B=Constant2Vector() {g=0.5, r=0.5}), B=@n53) {constB=0.5}
  Divide(A=Constant() {r=1}, B=Multiply(A=@n54, B=@n54))
  n55 = NamedRerouteUsage()
  DotProduct(A=@n55, B=Constant3Vector() {constant=<rgba 0 0 1 0>})
  n56 = FunctionInput() {inputType=FunctionInput_MaterialAttributes}
  n57 = GetMaterialAttributes(None=@n56) {attributeGetTypes [
    #1
    #2
    #3
    #4
    #5
    #6
   ]}
  n58 = NamedRerouteUsage()
  n59 = Saturate(None=Divide(A=PixelDepth().R, B=ScalarParameter() {defaultValue=60000, parameterName=DistanceGoldAdd DistFade}))
  n60 = Saturate(None=Add(A=NamedRerouteUsage()) {constB=0.3})
  n61 = Multiply(
   A=@n58
   B=LinearInterpolate(
    A=ScalarParameter() {defaultValue=1.8, parameterName=Mobile Emissvie}
    B=ScalarParameter() {defaultValue=5, parameterName=Mobile EmissvieAdd}
    Alpha=Multiply(A=@n59, B=@n60)
   ) {constA=1}
  )
  n62 = NamedRerouteUsage()
  n63 = NamedRerouteDeclaration(None=@n57.Normal) {name=OriginalNormal, nodeColor=<rgba 1 0 0.7031 1>}
  n64 = NamedRerouteUsage()
  n65 = ScalarParameter() {defaultValue=-10, parameterName=GoldDistanceNormal}
  n66 = SetMaterialAttributes!(
   MaterialAttributes
   BaseColor=NamedRerouteUsage()
   Metallic=ScalarParameter() {defaultValue=1, group=Gold Liquid, parameterName=Gold_Metallic}
   Opacity=StaticSwitch(
    True=ScalarParameter() {defaultValue=1, group=Gold Liquid, parameterName=Gold_Opacity_Mobile}
    False=ScalarParameter() {defaultValue=1, group=Gold Liquid, parameterName=Gold_Opacity}
    Value=NamedRerouteUsage()
   )
   Roughness=ScalarParameter() {defaultValue=1.5, group=Gold Liquid, parameterName=Gold_Roughness}
   EmissiveColor=Add(A=@n57.EmissiveColor, B=StaticSwitch(True=@n61, False=@n58, Value=@n62))
   Normal=MaterialFunctionCall(
    Normal=MaterialFunctionCall(BaseNormal=@n63, AdditionalNormal=@n64) {materialFunction=&BlendAngleCorrectedNormals}.Result
    Flatness=LinearInterpolate(B=@n65, Alpha=@n59) {constB=-5}
   ) {materialFunction=&FlattenNormal}.Result
   Specular=ScalarParameter() {group=Gold Liquid, parameterName=Gold_Specular}
   WorldPositionOffset=@n57.WorldPositionOffset
   OpacityMask=@n57.WorldPositionOffset
  ) {attributeSetTypes [
    #7
    #8
    #9
    #5
    #6
    #1
    #4
    #2
    #3
   ]}
  n67 = Distance(
   A=ComponentMask(None=VectorParameter() {group=Gold Liquid_Blend, parameterName=Gold_SphereMask_Location, defaultValue=<rgba 0 0 0 1>}.RGB) {g=true, r=true}
   B=WorldPosition() {worldPositionShaderOffset=WPT_ExcludeAllShaderOffsets}.XY
  )
  n68 = ScalarParameter() {defaultValue=1.3, group=Gold Liquid_Blend, parameterName=Gold_SphereMask_Radius}
  n69 = Saturate(None=Divide(
    A=Subtract(A=@n67, B=@n68)
    B=ScalarParameter() {defaultValue=1.3, group=Gold Liquid_Blend, parameterName=Gold_SphereMask_RadiusFalloffDistance}
   ))
  n70 = StaticSwitch(
   True=ScalarParameter() {defaultValue=0.05, parameterName=DistanceGoldAdd_Mobile}
   False=LinearInterpolate(
    A=ScalarParameter() {defaultValue=0.05, parameterName=DistanceGoldAdd}
    B=ScalarParameter() {defaultValue=0.05, parameterName=DistanceGoldAdd_B}
    Alpha=NamedRerouteUsage()
   )
   Value=NamedRerouteUsage()
  )
  n71 = Saturate(None=Add(A=LinearInterpolate(B=@n70, Alpha=@n59), B=NamedRerouteUsage()))
  n72 = StaticSwitchParameter(True=Multiply(A=@n69, B=@n71), False=@n71) {group=Gold Liquid_Blend, parameterName=UseGold_SphereMask_Blend}
  FunctionOutput(None=StaticSwitchParameter(True=BlendMaterialAttributes(A=@n57.MaterialAttributes, B=@n66, Alpha=@n72) {vertexAttributeBlendType=UseA}, False=@n56) {group=Gold Liquid, parameterName=Gold Liquid})
  FunctionOutput(None=VectorParameter() {group=Gold Liquid, parameterName=Gold_WaterColorMobile, defaultValue=<rgba 0.0774 0.1802 0.1979 1>}.RGB) {outputName=MobileWaterColor, sortPriority=50}
  FunctionOutput(None=StaticSwitchParameter(True=@n69, False=Constant() {r=1}) {group=Gold Liquid_Blend, parameterName=UseGold_SphereMask_Blend}) {outputName=SphereMask, sortPriority=99}
  n73 = VectorParameter() {group=Gold Liquid, parameterName=Gold_Color, defaultValue=<rgba 2.1976 1.1806 0.3777 1>}
  n74 = ScalarParameter() {defaultValue=10, group=Gold Liquid, parameterName=Gold_NormalFresnel}
  n75 = Power(Base=OneMinus(None=NamedRerouteUsage()), Exp=@n74)
  LinearInterpolate(
   A=@n73.RGB
   B=VectorParameter() {group=Gold Liquid, parameterName=Gold_Color_2, defaultValue=<rgba 2.1976 1.1806 0.3777 1>}.RGB
   Alpha=Saturate(None=@n75)
  )
  LinearInterpolate(B=@n74, Alpha=Add(A=ComponentMask(None=@n55) {r=true}, B=ComponentMask(None=@n55) {g=true}))
  n76 = Rotator(Coordinate=@n47, Time=Multiply(A=ComponentMask(None=@n49) {r=true}) {constB=33.3})
  MaterialFunctionCall(UVs=@n76, Texture Scale=@n53) {materialFunction=&ScaleUVsByCenter}
  n77 = Rotator(Coordinate=@n37, Time=Multiply(A=ComponentMask(None=@n39) {r=true}) {constB=33.3})
  MaterialFunctionCall(UVs=@n77, Texture Scale=@n43) {materialFunction=&ScaleUVsByCenter}
  n78 = TextureObject() {samplerType=SAMPLERTYPE_Masks, texture=&GoldLiquid_M}
  n79 = ComponentMask(None=NamedRerouteUsage()) {g=true, r=true}
  n80 = ScalarParameter() {defaultValue=3773.1267, group=Flowmaps, parameterName=Gold_FlowmapWaveNormalScale}
  n81 = ScalarParameter() {defaultValue=0.1828, group=Flowmaps, parameterName=GoldPhase Fade Time}
  n82 = Divide(A=Multiply(A=ScalarParameter() {defaultValue=1.1102, parameterName=Gold Final Flow Mult}, B=Divide(A=@n79, B=@n80)), B=@n81)
  n83 = VectorParameter() {group=Flowmaps, parameterName=GOLDFlowmap UV2 Offset and Scale, defaultValue=<rgba 1 1 1 0>}
  n84 = MaterialFunctionCall(Time=MaterialFunctionCall() {materialFunction=/Water/Materials/Functions/GetWaterTime}.WaterTime, Speed=@n81) {materialFunction=&TimeWithSpeedVariable}
  MaterialFunctionCall!(
   Diffuse=@n78
   Normal Map
   "Flow Vector Map (see tooltip)"=@n82
   UV texture 2 offset and Scale=@n83.RGB
   "Time (see tooltip)"=@n84.Time
   UVs=NamedRerouteUsage()
  ) {materialFunction=/FortWater/Materials/Functions/WaterFlowMap}
  NamedRerouteDeclaration(None=FunctionInput() {bUsePreviewValueAsDefault=true, inputName=Transition, inputType=FunctionInput_Scalar, sortPriority=8}) {name=TransitionMask, nodeColor=<rgba 0.0078 0 1 1>}
  NamedRerouteDeclaration(None=@n21) {name=LargeMask, nodeColor=<rgba 1 0 0.2578 1>}
  n85 = FunctionInput() {bUsePreviewValueAsDefault=true, inputName=Velocity, sortPriority=2, previewValue=<1 0 0 1>}
  NamedRerouteDeclaration(None=Multiply(A=NamedRerouteDeclaration(None=@n85) {name=InputVelocity, nodeColor=<rgba 0.8281 1 0 1>}, B=NamedRerouteUsage())) {name=VelocityScaled, nodeColor=<rgba 0 1 0.4609 1>}
  NamedRerouteDeclaration(None=FunctionInput(Preview=StaticBool()) {bUsePreviewValueAsDefault=true, inputName=IsTranslucentMat, inputType=FunctionInput_StaticBool, sortPriority=120}) {name=IsTranslucentMat, nodeColor=<rgba 1 0 0.5859 1>}
  n86 = Time()
  n87 = ScalarParameter() {defaultValue=0.8, group=Gold Liquid, parameterName=Gold_Particles NormalRotation}
  n88 = TextureSample(UVs=Divide(A=@n2, B=ScalarParameter() {defaultValue=250, group=Gold Liquid, parameterName=Gold_Normal Size})) {
   samplerSource=SSM_Wrap_WorldGroupSettings
   samplerType=SAMPLERTYPE_Normal
   texture=&/GoldLiquid_normal
  }
  n89 = Add(
   A=RotateAboutAxis(
    NormalizedRotationAxis=Constant3Vector() {constant=<rgba 0 0 1 0>}
    RotationAngle=Multiply(A=@n86, B=@n87)
    PivotPoint=Constant3Vector()
    Position=@n88.RGB
   )
   B=@n88.RGB
  )
  n90 = QualitySwitch(Default=@n89, Low=@n88.RGB, High=@n89, Medium=@n88.RGB, Epic=@n89)
  n91 = ScalarParameter() {defaultValue=-1.8, group=Gold Liquid, parameterName=Gold_NormalFlatten}
  n92 = ScalarParameter() {defaultValue=0.5, group=Gold Liquid, parameterName=Gold_NormalFlattenRiver}
  n93 = NamedRerouteUsage()
  n94 = StaticSwitch(
   True=StaticSwitchParameter(True=LinearInterpolate(A=@n91, B=@n92, Alpha=@n93), False=@n92) {parameterName=River Transition}
   False=@n91
   Value=NamedRerouteUsage()
  )
  n95 = Saturate(None=Divide(A=PixelDepth().R, B=ScalarParameter() {defaultValue=3000, parameterName=DistanceGoldDetailNormalDistFade}))
  NamedRerouteDeclaration(None=MaterialFunctionCall(
    Normal=ShadingPathSwitch(Default=@n90, Mobile=@n88.RGB) {desc=Migrated from FeatureLevelSwitch}
    Flatness=LinearInterpolate(A=@n94, Alpha=@n95) {constA=1}
   ) {materialFunction=&FlattenNormal}.Result) {name=GoldNormal, nodeColor=<rgba 0 0.1328 1 1>}
  n96 = PixelDepth()
  n97 = ScalarParameter() {defaultValue=2000, parameterName=GoldEmissive DistFadeoffset}
  n98 = Saturate(None=Divide(A=Subtract(A=@n96.R, B=@n97), B=ScalarParameter() {defaultValue=4000, parameterName=GoldEmissive DistFade}))
  n99 = TextureSample!(UVs=@n76, Tex, "DDX(UVs)"=DDX(Value=@n46), "DDY(UVs)"=DDY(Value=@n46), Apply View MipBias) {mipValueMode=TMVM_Derivative, samplerType=SAMPLERTYPE_LinearGrayscale, texture=&/Gold_Flare_M}
  n100 = Multiply(A=@n99.R, B=@n52)
  n101 = ComponentMask(None=@n51) {g=true}
  n102 = Power(
   Base=Multiply(A=Multiply(A=@n100, B=@n101), B=ComponentMask(None=@n51) {b=true})
   Exp=LinearInterpolate(
    A=ScalarParameter() {defaultValue=50, parameterName=GoldEmissive Sparkleh}
    B=ScalarParameter() {defaultValue=1.5, parameterName=GoldEmissive Sparkle}
    Alpha=Saturate(None=@n53)
   )
  )
  n103 = Multiply(
   A=NamedRerouteDeclaration(None=OneMinus(None=@n98)) {name=PixelDepthFadeClose, nodeColor=<rgba 0.5938 0 1 1>}
   B=Multiply(A=ScalarParameter() {defaultValue=30, group=Gold Liquid, parameterName=Gold_Emissive_Sparkle}, B=Saturate(None=@n102))
  )
  n104 = Divide(A=PixelDepth().R, B=ScalarParameter() {defaultValue=86000, parameterName=GoldEmissivelarge DistFade})
  n105 = TextureSample(UVs=@n77, "DDX(UVs)"=DDX(Value=@n36), "DDY(UVs)"=DDY(Value=@n36)) {mipValueMode=TMVM_Derivative, samplerType=SAMPLERTYPE_LinearGrayscale, texture=&Gold_Flare_M}
  n106 = Multiply(A=@n105.R, B=@n42)
  n107 = ComponentMask(None=@n41) {g=true}
  n108 = Power(
   Base=Multiply(A=Multiply(A=@n106, B=@n107), B=ComponentMask(None=@n41) {b=true})
   Exp=LinearInterpolate(
    A=ScalarParameter() {defaultValue=10, parameterName=GoldEmissive Sparkle_Large_h}
    B=ScalarParameter() {defaultValue=2.1, parameterName=GoldEmissive Sparkle_Large}
    Alpha=Saturate(None=@n43)
   )
  )
  n109 = Multiply(
   A=OneMinus(None=Saturate(None=@n104))
   B=Multiply(A=ScalarParameter() {defaultValue=50, group=Gold Liquid, parameterName=Gold_Emissive_Sparkle_Large}, B=Saturate(None=@n108))
  )
  n110 = StaticSwitch(
   True=@n109
   False=ShadingPathSwitch(Default=Add(A=@n103, B=@n109), Mobile=@n109) {desc=Migrated from FeatureLevelSwitch}
   Value=NamedRerouteUsage()
  )
  n111 = NamedRerouteDeclaration(None=@n73.RGB) {name=GoldBaseColor, nodeColor=<rgba 1 0 0.5391 1>}
  n112 = ScalarParameter() {defaultValue=0.075, group=Gold Liquid, parameterName=Gold_Emissive_Night}
  n113 = Multiply(
   A=@n111
   B=MaterialFunctionCall(
    Value_Morning=@n112
    Value_Day=ScalarParameter() {defaultValue=0.062, group=Gold Liquid, parameterName=Gold_Emissive_Day}
    Value_Evening=@n112
    Value_Night=@n112
   ) {materialFunction=&Scalar_Time_of_Day}.Result
  )
  n114 = Add(A=Multiply(A=@n110, B=@n111), B=@n113)
  n115 = VectorParameter() {parameterName=ReflectionGlints, defaultValue=<rgba 30 20.4612 9.5312 1>}
  n116 = AtmosphericLightVector()
  n117 = ReflectionVectorWS(CustomWorldNormal=PixelNormalWS())
  n118 = Multiply(A=Power(Base=DotProduct(A=@n116, B=@n117), Exp=ScalarParameter() {defaultValue=35.6438, parameterName=ReflectionGlint Power}), B=@n110)
  n119 = StaticSwitchParameter(True=Add(A=@n114, B=Multiply(A=@n115.RGB, B=@n118)), False=@n113) {defaultValue=true, parameterName=UseGoldGlints}
  n120 = NamedRerouteUsage()
  NamedRerouteDeclaration(None=EyeAdaptationInverse(LightValueInput=StaticSwitch(True=@n114, False=@n119, Value=@n120)).EyeAdaptationInverse) {name=GoldEmissive, nodeColor=<rgba 0.2187 1 0 1>}
  n121 = StaticSwitch(
   True=ScalarParameter() {defaultValue=5, group=Gold Liquid, parameterName=GoldOpacityFinalMobile}
   False=ScalarParameter() {defaultValue=1.3, group=Gold Liquid, parameterName=GoldOpacityFinal}
   Value=NamedRerouteUsage()
  )
  n122 = NamedRerouteUsage()
  n123 = LinearInterpolate(A=Multiply(A=NamedRerouteUsage(), B=@n24.B), Alpha=@n1)
  n124 = Saturate(None=@n123)
  n125 = Multiply(A=@n122, B=@n124)
  n126 = Saturate(None=LinearInterpolate(A=Multiply(A=@n34, B=@n33), Alpha=@n1))
  n127 = LinearInterpolate(A=@n126, B=@n124, Alpha=@n122)
  n128 = StaticBool()
  n129 = StaticSwitch(
   True=StaticSwitchParameter(True=StaticSwitchParameter(True=@n125, False=@n127) {parameterName=Ocean River Transition}, False=@n123) {parameterName=River Transition}
   False=@n126
   Value=NamedRerouteDeclaration(None=FunctionInput(Preview=@n128) {bUsePreviewValueAsDefault=true, inputName=IsRiverMaterial, inputType=FunctionInput_StaticBool, sortPriority=99}) {name=isRiver, nodeColor=<rgba 1 0 0.7969 1>}
  )
  n130 = NamedRerouteUsage()
  n131 = ScalarParameter() {defaultValue=62.299, group=Gold Liquid, parameterName=RiverGoldDepthFade}
  n132 = Multiply(A=@n129, B=Saturate(None=Divide(A=@n130, B=@n131)))
  NamedRerouteDeclaration(None=Saturate(None=Multiply(A=@n121, B=@n132))) {name=Final Mask Gold, nodeColor=<rgba 1 0 0.1406 1>}
  NamedRerouteDeclaration(None=@n24.B) {name=HugeMask, nodeColor=<rgba 0.8047 0 1 1>}
  n133 = AppendVector(
   A=ScalarParameter() {defaultValue=1, group=Gold Liquid, parameterName=Gold_RadialUV_X}
   B=ScalarParameter() {defaultValue=1, group=Gold Liquid, parameterName=Gold_RadialUV_Y}
  )
  n134 = ScalarParameter() {defaultValue=0.5, group=RadialUVs, parameterName=RadialUV_Center_X}
  n135 = ScalarParameter() {defaultValue=0.5, group=RadialUVs, parameterName=RadialUV_Center_Y}
  n136 = MaterialFunctionCall(Vector or UVs=Subtract(A=TextureCoordinate(), B=AppendVector(A=@n134, B=@n135))) {materialFunction=&VectorToRadialValue}
  n137 = ScalarParameter() {defaultValue=0.1, group=Gold Liquid, parameterName=Gold_RadialPan_X}
  n138 = ScalarParameter() {defaultValue=0.1, group=Gold Liquid, parameterName=Gold_RadialPan_Y}
  n139 = StaticSwitchParameter(
   True=Panner(Coordinate=Multiply(A=@n133, B=@n136.Radial Coordinates), Speed=AppendVector(A=@n137, B=@n138))
   False=WorldPosition() {worldPositionShaderOffset=WPT_ExcludeAllShaderOffsets}.XY
  ) {group=Gold Liquid, parameterName=UseGoldRadialUVs}
  n140 = Normalize(VectorInput=Subtract(A=WorldPosition().XYZ, B=CameraPositionWS()))
  n141 = Transform(None=@n140) {transformSourceType=TRANSFORMSOURCE_World, transformType=TRANSFORM_Tangent}
  n142 = ScalarParameter() {defaultValue=-100, group=Gold Liquid, parameterName=Gold_DepthOffset}
  n143 = NamedRerouteUsage()
  n144 = DotProduct(A=@n140, B=@n143)
  n145 = Add(
   A=NamedRerouteDeclaration(None=@n139) {name=UV_noPrarallax, nodeColor=<rgba 1 0.6562 0 1>}
   B=Multiply(A=ComponentMask(None=@n141) {g=true, r=true}, B=Divide(A=@n142, B=@n144))
  )
  NamedRerouteDeclaration(None=ShadingPathSwitch(Default=QualitySwitch(Default=@n145, Low=@n139, High=@n139, Medium=@n139), Mobile=@n139) {desc=Migrated from FeatureLevelSwitch}) {name=UV_withParallax_A, nodeColor=<rgba 1 0.6562 0 1>}
  n146 = NamedRerouteUsage()
  n147 = ScalarParameter() {defaultValue=2557.1453, group=Gold Liquid, parameterName=Gold_FlowMap Size}
  n148 = MaterialFunctionCall(Float3=MaterialFunctionCall(
    Diffuse=@n78
    "Flow Vector Map (see tooltip)"=@n82
    UV texture 2 offset and Scale=@n83.RGB
    "Time (see tooltip)"=@n84.Time
    UVs=Divide(A=@n146, B=@n147)
   ) {materialFunction=&WaterFlowMap}.Diffuse) {materialFunction=&BreakOutFloat3Components}
  NamedRerouteDeclaration(None=Multiply(A=@n148.G, B=ScalarParameter() {defaultValue=1.7, group=Gold Liquid, parameterName=RiverGoldMult})) {name=RiverFlowMask, nodeColor=<rgba 0.125 0 1 1>}
  n149 = FunctionInput() {bUsePreviewValueAsDefault=true, inputName=Depth, inputType=FunctionInput_Scalar, sortPriority=4, previewValue=<1 0 0 1>}
  NamedRerouteDeclaration(None=Max(A=@n149) {constB=0}) {name=Input Depth, nodeColor=<rgba 0.4531 0 1 1>}
  n150 = ScalarParameter() {defaultValue=256, group=Flowmaps, parameterName=Target Depth for Max Velocity Flow}
  MaterialExpressionDivide_8 = Divide(
   A=NamedRerouteUsage()
   B=AppendVector(A=Divide(A=@n150), B=ScalarParameter() {defaultValue=128, group=Flowmaps, parameterName=Target Depth for Flowmap Waves})
  )
  n151 = Custom!(x=@MaterialExpressionDivide_8) {
   code="return 1-exp(-x);"
   description=exp
   outputType=CMOT_Float2
   inputs [{inputName=x, input {expression=&self:MaterialExpressionDivide_8}}]
  }
  n152 = ComponentMask(None=@n151) {r=true}
  NamedRerouteDeclaration(None=@n152) {name=TargetDepthForMaxVelocity, nodeColor=<rgba 1 0 0.7734 1>}
  n153 = ComponentMask(None=@n151) {g=true}
  NamedRerouteDeclaration(None=@n153) {name=TargetDepth For Flowmap Waves, nodeColor=<rgba 0.2891 1 0 1>}
  NamedRerouteUsage()
  NamedRerouteUsage()
  NamedRerouteUsage()
  NamedRerouteUsage()
  NamedRerouteUsage()
  OneMinus(None=@n148.G)
  OneMinus(None=@n69)
  PixelNormalWS()
  Power()
  Reroute()
  Reroute(None=@n152)
  Reroute(None=@n153)
  Saturate(None=@n75)
  ScalarParameter() {defaultValue=0.08, group=Gold Liquid, parameterName=Gold_Emissive_Evening}
  ScalarParameter() {defaultValue=180, group=Gold Liquid, parameterName=Gold_Particles Size2}
  StaticSwitch(True=@n126, False=@n129, Value=NamedRerouteUsage())
  Subtract(A=@n37)
  Transform(None=@n143)
  VertexNormalWS()
  WorldPosition()
 ]}
